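; Battle Pass Season 23 Cipher quest computer terminal actor that changes its screen or mesh visuals after a player interacts with it.
; The post-interaction visuals event calls the parent, then creates a dynamic material instance on the post-interaction quest mesh, using the SignalControl parameter's primitive data index to pick the source material.
; It then writes a fixed value to that mesh's first custom primitive data float.
; BeginPlay, overlap, tick and construction add no logic.
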(event EventBeginPlay
  (|CallParentFunction) ; This node is disabled and will not be called.\nDrag off pins to build functionality.
)

(event Collision|EventActorBeginOverlap
  (params (OtherActor :type Actor%20Object%20Reference))
  (|CallParentFunction OtherActor) ; This node is disabled and will not be called.\nDrag off pins to build functionality.
)

(event EventTick
  (params (DeltaSeconds :type Float%20%28single-precision%29))
  (|CallParentFunction DeltaSeconds) ; This node is disabled and will not be called.\nDrag off pins to build functionality.
)

(event VisualState|Meshes|EventPostInteractionVisuals
  (params (PlayerController :type Fort%20Player%20Controller%20Object%20Reference))
  (|CallParentFunction PlayerController)
  (Rendering|Material|CreateDynamicMaterialInstance (Variables|Components|GetSM_Quest_PostInteraction) 0 (Rendering|Material|GetMaterial (Variables|Components|GetSM_Quest_PostInteraction) (Rendering|Material|GetCustomPrimitiveDataIndexforScalarParameter (Variables|Components|GetSM_Quest_PostInteraction) "SignalControl")) "None" :!class /Script/Engine.PrimitiveComponent)
  (Rendering|Material|SetCustomPrimitiveDataFloat (Variables|Components|GetSM_Quest_PostInteraction) 0 1.7))

(fn ConstructionScript
  (|CallParentFunction))
