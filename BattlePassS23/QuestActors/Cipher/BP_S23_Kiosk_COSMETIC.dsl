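; Cosmetic kiosk actor for what appears to be a Season 23 Cipher quest, playing a sign-removal animation and then removing itself.
; On BeginPlay it stores the mesh's relative position and rotation, calls spawn-sound-attached on the mesh and plays SignRemovedTimeline.
; Timeline height, rotation and scale tracks set the mesh's relative transform each update, without using the stored initial values.
; When the timeline finishes, the actor destroys itself.
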(event EventBeginPlay
  (bind _sm_kiosk (Variables|Default|GetSM_Kiosk))
  (Variables|Default|SetInitialPosition (Variables|/Script/Engine.SceneComponent|GetRelativeLocation _sm_kiosk))
  (Variables|Default|SetInitialRotation (Variables|/Script/Engine.SceneComponent|GetRelativeRotation _sm_kiosk))
  (Audio|SpawnSoundAttached 0 (Variables|Default|GetSM_Kiosk))
  (bind (_direction _height _rotation _scale) (Timeline|SignRemovedTimeline PlayFromStart)
    (:Update
      (Transformation|SetRelativeTransform (Variables|Default|GetSM_Kiosk) (Math|Transform|MakeTransform (Math|Vector|MakeVector 0.0 0.0 _height) (Math|Rotator|MakeRotator (.x _rotation) (.y _rotation) (.z _rotation)) _scale) false false))
    (:Finished
      (Actor|DestroyActor))))

(fn ConstructionScript)
