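# Master material for a Cipher quest screen in BattlePassS23, drawing a grid puzzle on an arcade cabinet display.
# Index parameters become 5x3 cell UV offsets via Fmod and Truncate, placing a pulsing green dot, up to four obstacles and a logo end point.
# Parameters blend in flashing red error and pulsing win images, styled with scanlines and LED pixels.
# World position offset math stretches the mesh per local axis, relative to object radius and scale.
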Material /BattlePassS23/QuestActors/Cipher/Materials/M_Quests_master
graph {
 properties {pixelDepthOffsetMode=PDOM_Legacy, usedShadingModels=Default Lit}
 nodes [
  n1 = NamedRerouteUsage()
  n2 = ScalarParameter() {parameterName=Use error}
  n3 = TextureCoordinate()
  n4 = AppendVector(
   A=ScalarParameter() {defaultValue=-0.01, parameterName=Global U Offset}
   B=ScalarParameter() {defaultValue=0.005, parameterName=Global V Offset}
  )
  n5 = ScalarParameter() {defaultValue=0.945, parameterName=Global U Scale}
  n6 = ScalarParameter() {defaultValue=0.945, parameterName=Global VScale}
  n7 = NamedRerouteDeclaration(None=MaterialFunctionCall(UVs=Add(A=@n3, B=@n4), Texture Scale=AppendVector(A=@n5, B=@n6)) {materialFunction=&ScaleUVsByCenter}.UVs) {name=Global UV Scale, nodeColor=<rgba 1 0.7266 0 1>}
  n8 = Constant() {r=5}
  n9 = Constant() {r=3}
  n10 = AppendVector(A=@n8, B=@n9)
  n11 = Add(
   A=TextureSample(UVs=Multiply(A=@n7, B=@n10)) {samplerSource=SSM_Wrap_WorldGroupSettings, texture=/BattlePassS23/QuestActors/Cipher/Textures/T_Quests}.R
   B=TextureSample() {texture=/Game/Environments/Apollo/Sets/Commerce/Props/Textures/T_Commerce_ArcadeCabinet_02_Screen_D}.RGB
  )
  n12 = Constant3Vector() {constant=<rgba 0 1 0 0>}
  n13 = NamedRerouteUsage()
  n14 = LinearInterpolate(
   A=LinearInterpolate(A=LinearInterpolate(A=@n11, B=@n12, Alpha=@n13), B=Constant3Vector() {constant=<rgba 1 1 1 0>}, Alpha=NamedRerouteUsage())
   B=NamedRerouteUsage()
   Alpha=NamedRerouteUsage()
  )
  n15 = Multiply(A=Saturate(None=@n14)) {constB=0.03}
  n16 = LinearInterpolate(
   A=Add(A=LinearInterpolate(B=@n1, Alpha=@n2), B=LinearInterpolate(A=@n14, B=@n15, Alpha=@n2))
   B=NamedRerouteUsage()
   Alpha=ScalarParameter() {parameterName=Use Win}
  )
  n17 = TextureCoordinate() {uTiling=200, vTiling=200}
  Add(
   A=Multiply(A=@n16) {constB=0.1}
   B=LinearInterpolate(A=TextureSample(UVs=@n17) {texture=/BattlePassS23/QuestActors/Cipher/Textures/T_LEDPixel_Quests}.RGB, Alpha=Constant() {r=0.99}) {constB=0}
  )
  n18 = Divide(A=Subtract(A=ObjectPositionWS(), B=WorldPosition().XYZ), B=ObjectRadius())
  n19 = MaterialFunctionCall() {materialFunction=&ObjectScale}
  n20 = ComponentMask(None=Transform(None=Multiply(A=@n18, B=@n19.Scale X)) {transformSourceType=TRANSFORMSOURCE_World, transformType=TRANSFORM_Local}) {r=true}
  n21 = Constant() {r=0.3}
  n22 = Multiply(A=@n18, B=@n19.Scale Y)
  n23 = Add(
   A=ComponentMask(None=Transform(None=@n22) {transformSourceType=TRANSFORMSOURCE_World, transformType=TRANSFORM_Local}) {g=true}
   B=ScalarParameter() {defaultValue=0.064, parameterName=Sclae_offset}
  )
  n24 = ScalarParameter() {defaultValue=113, parameterName=Scale_Sterngth}
  n25 = ComponentMask(None=Transform(None=Multiply(A=@n18, B=@n19.Scale Z)) {transformSourceType=TRANSFORMSOURCE_World, transformType=TRANSFORM_Local}) {b=true}
  n26 = Constant() {r=1}
  n27 = Transform(None=MaterialFunctionCall(X=Multiply(A=@n20, B=@n21) {constB=1.1}, Y=Multiply(A=@n23, B=@n24), Z=Multiply(A=@n25, B=@n26) {constB=1.1}) {materialFunction=&MakeFloat3}.Result) {transformSourceType=TRANSFORMSOURCE_Local}
  Add(A=@n27)
  n28 = MaterialFunctionCall!(
   ScanLine_Tiling_1=ScalarParameter() {defaultValue=60, parameterName=scanlines}
   Scanline_PannerSpeed_1
   ScanLine_Tiling=ScalarParameter() {defaultValue=120, parameterName=scanlines2}
   Scanline_PannerSpeed
   UVs=TextureCoordinate()
  ) {materialFunction=/Game/Weapons/WeaponSkins/MaterialFunction/ScanLines}
  Multiply(
   A=LinearInterpolate(Alpha=TextureSample() {texture=&T_Commerce_ArcadeCabinet_02_Screen_D}.R) {constA=0.4}
   B=Multiply(A=LinearInterpolate(Alpha=@n28.Result) {constA=0.9}, B=@n16)
  )
  n29 = NamedRerouteUsage()
  n30 = AppendVector(A=NamedRerouteUsage(), B=Constant() {r=1})
  n31 = NamedRerouteUsage()
  n32 = ScalarParameter() {defaultValue=10, parameterName=Green dot index}
  n33 = Constant() {r=5}
  n34 = Fmod(A=@n32, B=@n33)
  n35 = Subtract(A=Multiply(A=@n31) {constB=2}, B=Multiply(A=@n31, B=If(A=@n34, "A > B"=@n34, "A == B"=@n33, "A < B"=@n34)))
  n36 = Divide(A=Subtract(A=@n32), B=@n33) {constB=5}
  n37 = Subtract(A=NamedRerouteUsage(), B=Multiply(A=NamedRerouteUsage(), B=Truncate!(None=@n36) {is=class MaterialExpressionTruncate}))
  n38 = MaterialFunctionCall(UVs=Add(A=Multiply(A=@n29, B=@n30), B=AppendVector(A=@n35, B=@n37)), Texture Scale=Constant() {r=0.464}) {materialFunction=&ScaleUVsByCenter}
  n39 = MaterialFunctionCall(Value=Multiply(A=Time()) {constB=0.5}) {materialFunction=&LinearSine}
  NamedRerouteDeclaration(None=Multiply(
    A=TextureSample(UVs=@n38.UVs) {samplerSource=SSM_Clamp_WorldGroupSettings, texture=&T_Quests}.G
    B=LinearInterpolate(Alpha=@n39.Linear Sine) {constA=8, constB=15}
   )) {name=Green dot index, nodeColor=<rgba 0 0.4609 1 1>}
  n40 = Multiply(A=NamedRerouteUsage(), B=AppendVector(A=NamedRerouteUsage(), B=Constant() {r=1}))
  n41 = NamedRerouteUsage()
  n42 = ScalarParameter() {defaultValue=2, parameterName=obstacle4 index}
  n43 = Constant() {r=5}
  n44 = Fmod(A=@n42, B=@n43)
  n45 = If(A=@n44, "A > B"=@n44, "A == B"=@n43, "A < B"=@n44)
  n46 = NamedRerouteUsage()
  n47 = Truncate(None=Divide(A=Subtract(A=@n42), B=@n43) {constB=5})
  n48 = AppendVector(A=Subtract(A=Multiply(A=@n41) {constB=2}, B=Multiply(A=@n41, B=@n45)), B=Subtract(A=NamedRerouteUsage(), B=Multiply(A=@n46, B=@n47)))
  n49 = TextureSample(UVs=MaterialFunctionCall(UVs=Add(A=@n40, B=@n48), Texture Scale=NamedRerouteUsage()) {materialFunction=&ScaleUVsByCenter}.UVs) {samplerSource=SSM_Clamp_WorldGroupSettings, texture=&T_Quests}
  n50 = Constant()
  n51 = StaticBoolParameter() {defaultValue=true, parameterName=Use obstacle4}
  n52 = NamedRerouteUsage()
  n53 = Constant() {r=1}
  n54 = NamedRerouteUsage()
  n55 = Multiply(A=@n54) {constB=2}
  n56 = ScalarParameter() {defaultValue=1, parameterName=obstacle3 index}
  n57 = Constant() {r=5}
  n58 = Fmod(A=@n56, B=@n57)
  n59 = Multiply(A=@n54, B=If(A=@n58, "A > B"=@n58, "A == B"=@n57, "A < B"=@n58))
  n60 = NamedRerouteUsage()
  n61 = Subtract(A=@n56)
  n62 = Multiply(A=NamedRerouteUsage(), B=Truncate(None=Divide(A=@n61, B=@n57) {constB=5}))
  n63 = Add(A=Multiply(A=NamedRerouteUsage(), B=AppendVector(A=@n52, B=@n53)), B=AppendVector(A=Subtract(A=@n55, B=@n59), B=Subtract(A=@n60, B=@n62)))
  n64 = NamedRerouteUsage()
  n65 = StaticSwitch(
   True=TextureSample(UVs=MaterialFunctionCall(UVs=@n63, Texture Scale=@n64) {materialFunction=&ScaleUVsByCenter}.UVs) {samplerSource=SSM_Clamp_WorldGroupSettings, texture=&T_Quests}.B
   False=Constant()
   Value=StaticBoolParameter() {defaultValue=true, parameterName=Use obstacle3}
  )
  n66 = NamedRerouteUsage()
  n67 = AppendVector(A=NamedRerouteUsage(), B=Constant() {r=1})
  n68 = NamedRerouteUsage()
  n69 = ScalarParameter() {defaultValue=8, parameterName=obstacle2 index}
  n70 = Constant() {r=5}
  n71 = Fmod(A=@n69, B=@n70)
  n72 = Subtract(A=Multiply(A=@n68) {constB=2}, B=Multiply(A=@n68, B=If(A=@n71, "A > B"=@n71, "A == B"=@n70, "A < B"=@n71)))
  n73 = Divide(A=Subtract(A=@n69), B=@n70) {constB=5}
  n74 = Subtract(A=NamedRerouteUsage(), B=Multiply(A=NamedRerouteUsage(), B=Truncate(None=@n73)))
  n75 = MaterialFunctionCall(UVs=Add(A=Multiply(A=@n66, B=@n67), B=AppendVector(A=@n72, B=@n74)), Texture Scale=NamedRerouteUsage()) {materialFunction=&ScaleUVsByCenter}
  n76 = NamedRerouteUsage()
  n77 = Constant() {r=1}
  n78 = NamedRerouteUsage()
  n79 = Multiply(A=@n78) {constB=2}
  n80 = ScalarParameter() {defaultValue=5, parameterName=obstacle index}
  n81 = Constant() {r=5}
  n82 = Fmod(A=@n80, B=@n81)
  n83 = Multiply(A=@n78, B=If(A=@n82, "A > B"=@n82, "A == B"=@n81, "A < B"=@n82))
  n84 = NamedRerouteUsage()
  n85 = Subtract(A=@n80)
  n86 = Multiply(A=NamedRerouteUsage(), B=Truncate(None=Divide(A=@n85, B=@n81) {constB=5}))
  n87 = Add(A=Multiply(A=NamedRerouteUsage(), B=AppendVector(A=@n76, B=@n77)), B=AppendVector(A=Subtract(A=@n79, B=@n83), B=Subtract(A=@n84, B=@n86)))
  n88 = NamedRerouteUsage()
  n89 = Max(
   A=StaticSwitch(
    True=TextureSample(UVs=@n75.UVs) {samplerSource=SSM_Clamp_WorldGroupSettings, texture=&T_Quests}.B
    False=Constant()
    Value=StaticBoolParameter() {defaultValue=true, parameterName=Use obstacle2}
   )
   B=TextureSample(UVs=MaterialFunctionCall(UVs=@n87, Texture Scale=@n88) {materialFunction=&ScaleUVsByCenter}.UVs) {samplerSource=SSM_Clamp_WorldGroupSettings, texture=&T_Quests}.B
  )
  NamedRerouteDeclaration(None=Max(A=StaticSwitch(True=@n49.B, False=@n50, Value=@n51), B=Max(A=@n65, B=@n89))) {name=obstacle1 index, nodeColor=<rgba 0 0.4609 1 1>}
  NamedRerouteDeclaration(None=@n27) {name=WPO_Scale, nodeColor=<rgba 1 0.0234 0 1>}
  n90 = Multiply(A=NamedRerouteUsage(), B=AppendVector(A=NamedRerouteUsage(), B=Constant() {r=1}))
  n91 = NamedRerouteUsage()
  n92 = ScalarParameter() {defaultValue=15, parameterName=End point index}
  n93 = Constant() {r=5}
  n94 = Fmod(A=@n92, B=@n93)
  n95 = If(A=@n94, "A > B"=@n94, "A == B"=@n93, "A < B"=@n94)
  n96 = NamedRerouteUsage()
  n97 = Truncate(None=Divide(A=Subtract(A=@n92), B=@n93) {constB=5})
  n98 = AppendVector(A=Subtract(A=Multiply(A=@n91) {constB=2}, B=Multiply(A=@n91, B=@n95)), B=Subtract(A=NamedRerouteUsage(), B=Multiply(A=@n96, B=@n97)))
  n99 = TextureSample(UVs=MaterialFunctionCall(UVs=Add(A=@n90, B=@n98), Texture Scale=Constant() {r=0.45}) {materialFunction=&ScaleUVsByCenter}.UVs) {samplerSource=SSM_Clamp_WorldGroupSettings, texture=/BattlePassS23/QuestActors/Cipher/Textures/T_logo}
  NamedRerouteDeclaration(None=@n99.A) {name=End point alpha, nodeColor=<rgba 0.3359 1 0 1>}
  n100 = Power(Base=@n99.G) {constExponent=10}
  n101 = Panner(Coordinate=MaterialFunctionCall(UVs=TextureCoordinate() {uTiling=0.1, vTiling=4}, "Rotation Angle (0-1)"=Constant() {r=0.25}) {materialFunction=&CustomRotator}.Rotated Values) {speedX=-0.2}
  n102 = Multiply(
   A=Saturate(None=Multiply(A=@n100) {constB=100})
   B=LinearInterpolate(Alpha=TextureSample(UVs=@n101) {texture=&T_LEDPixel_Quests}.R) {constA=1, constB=35}
  )
  n103 = Constant3Vector() {constant=<rgba 1 0.26 0.7434 0>}
  NamedRerouteDeclaration(None=Add(A=Multiply(A=@n99.RGB) {constB=1.2}, B=Multiply(A=@n102, B=@n103))) {name=End point Color, nodeColor=<rgba 0 0.5078 1 1>}
  n104 = Constant3Vector()
  n105 = NamedRerouteUsage()
  n106 = AppendVector(A=NamedRerouteUsage(), B=Constant() {r=1})
  n107 = Constant() {r=-0.33}
  n108 = Constant()
  n109 = MaterialFunctionCall(Value=Multiply(A=Time()) {constB=0.5}) {materialFunction=&LinearSine}
  n110 = Saturate(None=Multiply(A=@n109.Linear Sine, B=@n109.Linear Sine))
  n111 = MaterialFunctionCall(UVs=Add(A=Multiply(A=@n105, B=@n106), B=AppendVector(A=@n107, B=@n108)), Texture Scale=LinearInterpolate(Alpha=@n110) {constA=1, constB=1.5}) {materialFunction=&ScaleUVsByCenter}
  n112 = Saturate(None=Add(A=@n110) {constB=0.2})
  n113 = Multiply(A=@n105, B=AppendVector(A=NamedRerouteUsage(), B=Constant() {r=1}))
  n114 = AppendVector(A=Constant() {r=-0.33}, B=Constant())
  n115 = Max(
   A=Multiply(
    A=TextureSample(UVs=@n111.UVs) {samplerSource=SSM_Clamp_WorldGroupSettings, texture=&T_logo}.RGBA
    B=LinearInterpolate(Alpha=@n112) {constA=1, constB=0}
   )
   B=TextureSample(UVs=Add(A=@n113, B=@n114)) {samplerSource=SSM_Clamp_WorldGroupSettings, texture=&T_logo}.RGBA
  )
  n116 = ComponentMask(None=@n115) {b=true, g=true, r=true}
  n117 = ComponentMask(None=@n115) {a=true}
  n118 = MaterialFunctionCall(Value=Multiply(A=Time()) {constB=0.5}) {materialFunction=&LinearSine}
  n119 = Saturate(None=Multiply(A=@n118.Linear Sine, B=@n118.Linear Sine))
  n120 = LinearInterpolate(Alpha=@n119) {constA=1, constB=4}
  n121 = LinearInterpolate(Alpha=@n119) {constA=1, constB=0.25}
  NamedRerouteDeclaration(None=Multiply(A=LinearInterpolate(A=@n104, B=@n116, Alpha=@n117), B=Multiply(A=@n120, B=@n121))) {name=Win, nodeColor=<rgba 0 1 0.5547 1>}
  n122 = Constant3Vector()
  n123 = Constant3Vector() {constant=<rgba 1 0 0 0>}
  n124 = NamedRerouteUsage()
  n125 = AppendVector(A=NamedRerouteUsage(), B=Constant() {r=1})
  n126 = Constant() {r=-0.33}
  n127 = Constant()
  n128 = TextureSample(UVs=Add(A=Multiply(A=@n124, B=@n125), B=AppendVector(A=@n126, B=@n127))) {samplerSource=SSM_Clamp_WorldGroupSettings, texture=&T_Quests}
  n129 = MaterialFunctionCall(Value=Multiply(A=Time()) {constB=0.8}) {materialFunction=&LinearSine}
  NamedRerouteDeclaration(None=Multiply(A=LinearInterpolate(A=@n122, B=@n123, Alpha=@n128.A), B=LinearInterpolate(Alpha=@n129.Linear Sine) {constB=2})) {name=Error}
  NamedRerouteDeclaration(None=ScalarParameter() {defaultValue=0.33, parameterName=obstacle Scale}) {name=obstacle Scale, nodeColor=<rgba 1 0.4453 0 1>}
  NamedRerouteDeclaration(None=Divide(A=@n8, B=@n9)) {name=Aspect ratio, nodeColor=<rgba 1 0 0.7266 1>}
  n130 = Divide(B=@n9)
  n131 = Divide(A=@n9)
  NamedRerouteDeclaration(None=Subtract(A=Multiply(A=@n130, B=@n131), B=Multiply(A=@n130) {constB=0.5})) {name=uv distance, nodeColor=<rgba 1 0 0.6328 1>}
  NamedRerouteDeclaration(None=@n130) {name=uv distance_1, nodeColor=<rgba 1 0 0.6328 1>}
  NamedRerouteUsage()
  TextureSample() {samplerType=SAMPLERTYPE_Normal, texture=/Game/Environments/Apollo/Sets/Commerce/Props/Textures/T_Commerce_ArcadeCabinet_02_Screen_N}
  TextureSample() {samplerType=SAMPLERTYPE_Masks, texture=/Game/Environments/Apollo/Sets/Commerce/Props/Textures/T_Commerce_ArcadeCabinet_02_Screen_S}
 ]
}
